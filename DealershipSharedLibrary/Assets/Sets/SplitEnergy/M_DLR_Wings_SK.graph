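# Masked, lit material for the SplitEnergy set's skeletal-mesh wings, giving them animated energy color and glowing metal that support painted color variants.
# Panned, distorted noise textures and a time-pulsed tear texture build the energy masks and emissive color, lerped between two wing colors.
# A UV-based mask separates wing from metal, where IsPainted picks pulsing variant color over default color.
# A Dissolve parameter offsets a remapped dissolve texture, stepped into a hard opacity mask.
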Material /DealershipSharedLibrary/Assets/Sets/SplitEnergy/M_DLR_Wings_SK
graph {
 properties {
  bUsedWithMeshDeformer=true
  bUsedWithNiagaraMeshParticles=true
  bUsedWithSkeletalMesh=true
  blendMode=BLEND_Masked
  pixelDepthOffsetMode=PDOM_Legacy
  usedShadingModels=Default Lit
 }
 nodes [
  Composite() {subgraphName=Painted Variant Wings}
  Composite()
  n1 = Constant() {r=0.1}
  ConstantBiasScale(None=Sine(None=MaterialFunctionCall(Speed=@n1) {materialFunction=&TimeWithSpeedVariable}.Time))
  MaterialFunctionCall(Emissive=NamedRerouteUsage()) {materialFunction=&MF_FX_EmissiveAdjustments}
  n2 = ScalarParameter() {defaultValue=8, group=Dissolve, parameterName=Strange}
  n3 = MaterialFunctionCall(
   Input=TextureSample() {texture=&/T_SplitEnergy_CustomDissolveWings_A}.R
   Input Low=Constant()
   Input High=Constant() {r=1}
   Target Low=OneMinus(None=@n2)
   Target High=@n2
  ) {materialFunction=&RemapValueRange}
  n4 = ScalarParameter() {defaultValue=12, group=Dissolve, parameterName=Remap_Dissolve}
  n5 = MaterialFunctionCall(
   Input=ScalarParameter() {parameterName=Dissolve}
   Input Low=Constant()
   Input High=Constant() {r=1}
   Target Low=OneMinus(None=@n4)
   Target High=@n4
  ) {materialFunction=&RemapValueRange}
  NamedRerouteDeclaration(None=Step(Y=Add(A=@n3.Result, B=@n5.Result)) {constX=0.8, constY=1}) {name=Scalar Reroute Opacity, nodeColor=<rgba 1 0 0.3281 1>}
  n6 = TextureSample(UVs=Panner(Coordinate=TextureCoordinate() {uTiling=3, vTiling=3}) {speedX=-1}) {texture=/DealershipSharedLibrary/FX/Textures/Diffuse/T_VCVFX_SpaceTear_D}
  n7 = TextureSample(UVs=Panner(Coordinate=TextureCoordinate() {uTiling=1.5, vTiling=1.5}) {speedX=-0.8, speedY=0.05}) {texture=/DealershipSharedLibrary/FX/Textures/Gradient_Masks/T_VCVFX_SphereMaskTimeOffset_A}
  n8 = Time()
  MaterialExpressionConstantClamp_2 = Saturate(None=Sine(None=@n8) {period=0.5})
  MaterialExpressionConstantClamp_3 = Saturate(None=Cosine(None=@n8) {period=0.5})
  MaterialExpressionConstantClamp_1 = Saturate(None=Add(A=Multiply(A=@n7.R, B=@MaterialExpressionConstantClamp_2), B=Multiply(A=@n7.B, B=@MaterialExpressionConstantClamp_3)))
  NamedRerouteDeclaration(None=Multiply(A=Multiply(A=@n6.RGB) {constA=10, constB=10}, B=@MaterialExpressionConstantClamp_1)) {name=Reroute Pulse Noise, nodeColor=<rgba 1 0.0703 0 1>}
  n9 = TextureCoordinate()
  n10 = Round(None=Add(A=ComponentMask(None=@n9) {g=true}) {constB=0.2})
  NamedRerouteDeclaration(None=Multiply(A=@n10, B=NamedRerouteUsage())) {name=Reroute mask color wings, nodeColor=<rgba 0 1 0.6953 1>}
  n11 = NamedRerouteUsage()
  n12 = NamedRerouteUsage()
  n13 = NamedRerouteUsage()
  n14 = Multiply(A=NamedRerouteUsage(), B=LinearInterpolate(Alpha=NamedRerouteUsage()) {constA=0.0277, constB=0.84, desc=Desaturated WingsColor})
  n15 = TextureCoordinate() {uTiling=4, vTiling=4}
  n16 = Power(
   Base=TextureSample(UVs=Panner(Coordinate=@n15) {speedX=-1, speedY=-0.35}) {texture=/DealershipSharedLibrary/FX/Textures/Noise/T_VCVFX_NoiseText_A}.RGB
   Exp=Constant() {r=0.35}
  )
  n17 = If(
   A=NamedRerouteUsage()
   "A > B"=@n14
   "A == B"=@n14
   "A < B"=LinearInterpolate(
    A=VectorParameter() {group=Emissive, parameterName=WingsColor01, defaultValue=<rgba 0.5333 0 0.9843 1>}.RGB
    B=VectorParameter() {group=Emissive, parameterName=WingsColor02, defaultValue=<rgba 0.2 0.2314 1 1>}.RGB
    Alpha=NamedRerouteDeclaration(None=@n16) {name=Reroute TEX Color2, nodeColor=<rgba 1 0.5391 0 1>}
   )
  ) {constB=0.5}
  NamedRerouteDeclaration(None=Add(A=Multiply(A=@n11, B=@n12), B=Multiply(A=@n13, B=@n17))) {name=EmissiveColor, nodeColor=<rgba 1 0.375 0 1>}
  n18 = MaterialFunctionCall!(Primary Color, Tertiary Color, "Secondary  Color") {materialFunction=/DealershipSharedLibrary/FX/Materials/MaterialFunctions/MF_DLR_PaintedVariantsData}
  NamedRerouteDeclaration(None=@n18.Primary RGB) {name=Reroute Color Variants 2, nodeColor=<rgba 0 0.5547 1 1>}
  n19 = TextureSample(UVs=Panner(Coordinate=TextureCoordinate()) {speedX=-0.8}) {texture=/DealershipSharedLibrary/FX/Textures/Noise/T_VCVFX_TrailMask_RGB}
  n20 = Add(A=TextureCoordinate() {uTiling=2, vTiling=4}, B=Multiply(A=ComponentMask(None=@n19.RGB) {g=true, r=true}) {constB=0.45})
  n21 = NamedRerouteDeclaration(None=TextureSample(UVs=Panner(Coordinate=@n20) {speedX=-0.5}) {texture=/DealershipSharedLibrary/FX/Textures/Noise/T_VCVFX_Boost_alpine_infinite_Noise_Heightmap_Alpha}.RGB) {name=Reroute Tex Mask, nodeColor=<rgba 1 0 0.0937 1>}
  n22 = Constant() {r=0.8}
  n23 = TextureCoordinate() {uTiling=2, vTiling=2}
  n24 = Panner(Coordinate=TextureCoordinate()) {speedX=0.3}
  n25 = Multiply(A=ComponentMask(None=TextureSample(UVs=@n24) {texture=&T_VCVFX_TrailMask_RGB}.RGB) {g=true, r=true}) {constB=0.3}
  n26 = TextureSample(UVs=Panner(Coordinate=Add(A=@n23, B=@n25)) {speedX=-0.8}) {texture=/DealershipSharedLibrary/FX/Textures/Noise/T_VCVFX_AuraNoise01_A}
  n27 = Constant() {r=0.4}
  n28 = Power(Base=Add(A=Power(Base=@n21, Exp=@n22), B=Multiply(A=@n26.RGB, B=@n27)), Exp=Constant() {r=2.3})
  n29 = Constant() {r=2}
  NamedRerouteDeclaration(None=Add(A=Multiply(A=@n28, B=@n29), B=NamedRerouteUsage())) {name=Reroute Mask Textures, nodeColor=<rgba 0 0.4141 1 1>}
  n30 = ScalarParameter() {parameterName=IsPainted}
  n31 = VectorParameter() {group=Painted Variant Override, parameterName=DefaultColor, sortPriority=0, defaultValue=<rgba 1 0.1498 0.9461 1>}
  n32 = MaterialFunctionCall(Primary Color=@n31.RGBA) {materialFunction=&MF_DLR_PaintedVariantsData}
  n33 = ComponentMask(None=@n32.Primary A) {a=true}
  n34 = Time()
  n35 = Saturate(None=Add(A=Sine(None=@n34) {period=1.6}, B=Constant() {r=0.5}))
  n36 = Multiply(
   A=NamedRerouteDeclaration(None=@n32.Primary RGB) {name=Reroute CustomColor RGB, nodeColor=<rgba 0.8984 0 1 1>}
   B=LinearInterpolate(
    A=Multiply(B=@n33) {constA=0.3881, constB=12}
    Alpha=NamedRerouteDeclaration(None=@n35) {name=RerouteMaskLerpMetalColors, nodeColor=<rgba 0 1 0.5781 1>}
   ) {constB=0.2766}
  )
  n37 = NamedRerouteUsage()
  n38 = VectorParameter() {parameterName=Color Wings Metal Default Secondary, defaultValue=<rgba 0.1975 0.2271 1 1>}
  n39 = NamedRerouteUsage()
  NamedRerouteDeclaration(None=If(
    A=NamedRerouteDeclaration(None=@n30) {name=Reroute IsPainted, nodeColor=<rgba 1 0.8906 0 1>}
    "A > B"=@n36
    "A == B"=@n36
    "A < B"=LinearInterpolate(A=@n37, B=@n38.RGB, Alpha=@n39)
   ) {constB=0.5}) {name=Reroute Color Metal Wings, nodeColor=<rgba 1 0 0.3047 1>}
  n40 = NamedRerouteUsage()
  n41 = OneMinus(None=@n10)
  NamedRerouteDeclaration(None=Multiply(A=Multiply(A=@n40, B=@n41), B=ScalarParameter() {defaultValue=1.5, group=Emissive, parameterName=MaskMetalGlow})) {name=Reroute BaseColor, nodeColor=<rgba 0.2188 0 1 1>}
  NamedRerouteUsage()
  NamedRerouteUsage()
  PinBase()
  PinBase() {reroutePins [
    Gold Variant 2 {expression=null}
    Black Variant 2 {expression=null}
    BurntSienna Variant 2 {expression=null}
    Cobalt Variant 2 {expression=null}
    Crimson Variant 2 {expression=null}
    ForestGreen Variant 2 {expression=null}
    Gray Variant 2 {expression=null}
    Lime Variant 2 {expression=null}
    Orange Variant 2 {expression=null}
    Pink Variant 2 {expression=null}
    Purple Variant 2 {expression=null}
    Saffron Variant 2 {expression=null}
    SkyBlue Variant 2 {expression=null}
    TitaniumWhite Variant 2 {expression=null}
   ]}
  PinBase()
  n42 = VectorParameter() {group=Painted Variant Override, parameterName=Gold, sortPriority=2, defaultValue=<rgba 1 1 1 1>}
  n43 = VectorParameter() {group=Painted Variant Override, parameterName=Black, sortPriority=2, defaultValue=<rgba 1 1 1 1>}
  n44 = VectorParameter() {group=Painted Variant Override, parameterName=BurntSienna 1, sortPriority=3, defaultValue=<rgba 1 1 1 1>}
  n45 = VectorParameter() {group=Painted Variant Override, parameterName=Cobalt, sortPriority=4, defaultValue=<rgba 1 1 1 1>}
  n46 = VectorParameter() {group=Painted Variant Override, parameterName=Crimson, sortPriority=5, defaultValue=<rgba 1 1 1 1>}
  n47 = VectorParameter() {group=Painted Variant Override, parameterName=ForestGreen, sortPriority=6, defaultValue=<rgba 1 1 1 1>}
  n48 = VectorParameter() {group=Painted Variant Override, parameterName=Grey, sortPriority=7, defaultValue=<rgba 1 1 1 1>}
  n49 = VectorParameter() {group=Painted Variant Override, parameterName=Lime, sortPriority=8, defaultValue=<rgba 1 1 1 1>}
  n50 = VectorParameter() {group=Painted Variant Override, parameterName=Orange, sortPriority=9, defaultValue=<rgba 1 1 1 1>}
  n51 = VectorParameter() {group=Painted Variant Override, parameterName=Pink, sortPriority=10, defaultValue=<rgba 1 1 1 1>}
  n52 = VectorParameter() {group=Painted Variant Override, parameterName=Purple, sortPriority=11, defaultValue=<rgba 1 1 1 1>}
  n53 = VectorParameter() {group=Painted Variant Override, parameterName=Saffron, sortPriority=10, defaultValue=<rgba 1 1 1 1>}
  n54 = VectorParameter() {group=Painted Variant Override, parameterName=SkyBlue, sortPriority=13, defaultValue=<rgba 1 1 1 1>}
  n55 = VectorParameter() {group=Painted Variant Override, parameterName=TitaniumWhite, sortPriority=14, defaultValue=<rgba 1 1 1 1>}
  PinBase!(
   PV1 Gold=@n42.RGBA
   PV2 Black=@n43.RGBA
   PV3 BurntSienna=@n44.RGBA
   PV Cobalt=@n45.RGBA
   PV Crimson=@n46.RGBA
   PV ForestGreen=@n47.RGBA
   PV Gray=@n48.RGBA
   PV Lime=@n49.RGBA
   PV Orange=@n50.RGBA
   PV Pink=@n51.RGBA
   PV Purple=@n52.RGBA
   PV Saffron=@n53.RGBA
   PV SkyBlue=@n54.RGBA
   PV Titanium White=@n55.RGBA
  ) {reroutePins [
    PV1 Gold {expression=&self:MaterialExpressionReroute_13}
    PV2 Black {expression=&self:MaterialExpressionReroute_14}
    PV3 BurntSienna {expression=&self:MaterialExpressionReroute_15}
    PV Cobalt {expression=&self:MaterialExpressionReroute_16}
    PV Crimson {expression=&self:MaterialExpressionReroute_17}
    PV ForestGreen {expression=&self:MaterialExpressionReroute_18}
    PV Gray {expression=&self:MaterialExpressionReroute_19}
    PV Lime {expression=&self:MaterialExpressionReroute_20}
    PV Orange {expression=&self:MaterialExpressionReroute_21}
    PV Pink {expression=&self:MaterialExpressionReroute_22}
    PV Purple {expression=&self:MaterialExpressionReroute_23}
    PV Saffron {expression=&self:MaterialExpressionReroute_24}
    PV SkyBlue {expression=&self:MaterialExpressionReroute_25}
    PV Titanium White {expression=&self:MaterialExpressionReroute_26}
   ]}
  n56 = VectorParameter() {group=Painted Variant Override 2, parameterName=Pink Variant 2, sortPriority=10, defaultValue=<rgba 1 1 1 1>}
  Reroute(None=@n56.RGBA)
  n57 = VectorParameter() {group=Painted Variant Override 2, parameterName=Purple Variant 2, sortPriority=11, defaultValue=<rgba 1 1 1 1>}
  Reroute(None=@n57.RGBA)
  Reroute(None=@n18.Primary RGB)
  Reroute()
  Reroute(None=ScalarParameter() {group=Painted Variant Override, parameterName=ColorVariantID})
  MaterialExpressionReroute_13 = Reroute(None=@n42.RGBA)
  MaterialExpressionReroute_14 = Reroute(None=@n43.RGBA)
  MaterialExpressionReroute_15 = Reroute(None=@n44.RGBA)
  MaterialExpressionReroute_16 = Reroute(None=@n45.RGBA)
  MaterialExpressionReroute_17 = Reroute(None=@n46.RGBA)
  MaterialExpressionReroute_18 = Reroute(None=@n47.RGBA)
  MaterialExpressionReroute_19 = Reroute(None=@n48.RGBA)
  MaterialExpressionReroute_20 = Reroute(None=@n49.RGBA)
  MaterialExpressionReroute_21 = Reroute(None=@n50.RGBA)
  MaterialExpressionReroute_22 = Reroute(None=@n51.RGBA)
  MaterialExpressionReroute_23 = Reroute(None=@n52.RGBA)
  MaterialExpressionReroute_24 = Reroute(None=@n53.RGBA)
  MaterialExpressionReroute_25 = Reroute(None=@n54.RGBA)
  MaterialExpressionReroute_26 = Reroute(None=@n55.RGBA)
  n58 = VectorParameter() {group=Painted Variant Override 2, parameterName=Saffron Variant 2, sortPriority=10, defaultValue=<rgba 1 1 1 1>}
  Reroute(None=@n58.RGBA)
  n59 = VectorParameter() {group=Painted Variant Override 2, parameterName=SkyBlue Variant 2, sortPriority=13, defaultValue=<rgba 1 1 1 1>}
  Reroute(None=@n59.RGBA)
  n60 = VectorParameter() {group=Painted Variant Override 2, parameterName=TitaniumWhite Variant 2, sortPriority=14, defaultValue=<rgba 1 1 1 1>}
  Reroute(None=@n60.RGBA)
  VectorParameter() {group=Painted Variant Override 2, parameterName=BurntSienna Variant 2, sortPriority=3, defaultValue=<rgba 1 1 1 1>}
  VectorParameter() {group=Painted Variant Override 2, parameterName=Black Variant 2, sortPriority=2, defaultValue=<rgba 1 1 1 1>}
  VectorParameter() {group=Painted Variant Override 2, parameterName=Cobalt Variant 2, sortPriority=4, defaultValue=<rgba 1 1 1 1>}
  VectorParameter() {group=Painted Variant Override 2, parameterName=Crimson Variant 2, sortPriority=5, defaultValue=<rgba 1 1 1 1>}
  VectorParameter() {group=Painted Variant Override 2, parameterName=ForestGreen Variant 2, sortPriority=6, defaultValue=<rgba 1 1 1 1>}
  VectorParameter() {group=Painted Variant Override 2, parameterName=Grey Variant 2, sortPriority=7, defaultValue=<rgba 1 1 1 1>}
  VectorParameter() {group=Painted Variant Override 2, parameterName=Lime Variant 2, sortPriority=8, defaultValue=<rgba 1 1 1 1>}
  VectorParameter() {group=Painted Variant Override 2, parameterName=Orange Variant 2, sortPriority=9, defaultValue=<rgba 1 1 1 1>}
  VectorParameter() {group=Painted Variant Override 2, parameterName=Gold Variant 2, sortPriority=2, defaultValue=<rgba 1 1 1 1>}
 ]
}
