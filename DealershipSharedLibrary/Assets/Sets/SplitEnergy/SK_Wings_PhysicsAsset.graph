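# Physics asset for the SK_Wings skeletal mesh in the SplitEnergy set, modelling three mirrored left/right wing pairs as capsule bodies.
# Each wing is a chain of bones joined in sequence by constraints with uniform 45-degree swing and twist limit values.
# The W2 and W3 chains attach to the tiny Bone_008 body, while the W1 chains have no constraint to it.
# The W3 chains have no B2 body, so B3 links directly to B1.
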PhysicsAsset /DealershipSharedLibrary/Assets/Sets/SplitEnergy/SK_Wings_PhysicsAsset
bodies {
 Bone_008 [{shapeType=Capsule, length=0.505, radius=0.505, rotation=<rot 0 -90 0>}]
 W1B1-L [{shapeType=Capsule, length=43.8703, radius=40.4219, center=<2.4436 -44.0327 -1.0582>, rotation=<rot 0 26.9006 -83.7113>}]
 W1B1-R [{shapeType=Capsule, length=43.8703, radius=40.4219, center=<-2.4436 -44.0327 -1.0582>, rotation=<rot 0 -26.9006 -83.7113>}]
 W1B2-L [{shapeType=Capsule, length=67.3848, radius=6.8525, center=<3.9457 -36.6445 -0.602>, rotation=<rot 0 86.7768 -70.2845>}]
 W1B2-R [{shapeType=Capsule, length=67.3848, radius=6.8525, center=<-3.9457 -36.6445 -0.602>, rotation=<rot 0 -86.7768 -70.2845>}]
 W1B3-L [{shapeType=Capsule, length=78.8391, radius=6.8416, center=<0.2522 -38.3781 -1.9565>, rotation=<rot 0 94.724 -72.6653>}]
 W1B3-R [{shapeType=Capsule, length=78.8391, radius=6.8416, center=<-0.2521 -38.3781 -1.9565>, rotation=<rot 0 -94.724 -72.6654>}]
 W1B4-L [{shapeType=Capsule, length=76.1362, radius=29.781, center=<-4.8026 -16.8174 -3.8838>, rotation=<rot 0 90.8338 -70.7055>}]
 W1B4-R [{shapeType=Capsule, length=76.1362, radius=29.781, center=<4.8027 -16.8174 -3.8838>, rotation=<rot 0 -90.8338 -70.7055>}]
 W1B5-L [{shapeType=Capsule, length=66.5268, radius=25.475, center=<-10.0308 -25.3022 -3.8273>, rotation=<rot 0 96.6444 -79.0929>}]
 W1B5-R [{shapeType=Capsule, length=66.5268, radius=25.4749, center=<10.0309 -25.302 -3.8294>, rotation=<rot 0 -96.6433 -79.0923>}]
 W1B6-L [{shapeType=Capsule, length=60.9241, radius=45.798, center=<-10.2034 -58.6789 -1.8842>, rotation=<rot 0 172.3943 -88.9947>}]
 W1B6-R [{shapeType=Capsule, length=60.9241, radius=45.798, center=<10.2053 -58.678 -1.9008>, rotation=<rot 0 -172.3934 -88.9839>}]
 W2B1-L [{shapeType=Capsule, length=80.9341, radius=45.8068, center=<16.7723 -74.4223 0.7749>, rotation=<rot 0 32.5684 -87.4174>}]
 W2B1-R [{shapeType=Capsule, length=80.9341, radius=45.8068, center=<-16.7723 -74.4223 0.7749>, rotation=<rot 0 -32.5684 -87.4174>}]
 W2B2-L [{shapeType=Capsule, length=75.5672, radius=7.662, center=<6.8444 -46.8048 -0.7572>, rotation=<rot 0 86.0395 -85.8992>}]
 W2B2-R [{shapeType=Capsule, length=75.5672, radius=7.662, center=<-6.8444 -46.8048 -0.7573>, rotation=<rot 0 -86.0395 -85.8992>}]
 W2B3-L [{shapeType=Capsule, length=88.1614, radius=7.6606, center=<3.0582 -50.414 -0.6251>, rotation=<rot 0 95.7456 -86.0869>}]
 W2B3-R [{shapeType=Capsule, length=88.1614, radius=7.6606, center=<-3.0582 -50.414 -0.6251>, rotation=<rot 0 -95.7456 -86.0869>}]
 W2B4-L [{shapeType=Capsule, length=76.9333, radius=2.0957, center=<-11.9038 -48.9694 2.8606>, rotation=<rot 0 81.008 -86.7483>}]
 W2B4-R [{shapeType=Capsule, length=76.9334, radius=2.0957, center=<11.9039 -48.9694 2.8606>, rotation=<rot 0 -81.008 -86.7483>}]
 W2B5-L [{shapeType=Capsule, length=83.3797, radius=24.0546, center=<-11.4217 -15.774 -2.9733>, rotation=<rot 0 84.6118 -86.1074>}]
 W2B5-R [{shapeType=Capsule, length=83.3797, radius=24.0547, center=<11.4218 -15.774 -2.9733>, rotation=<rot 0 -84.6117 -86.1075>}]
 W2B6-L [{shapeType=Capsule, length=61.9701, radius=28.5378, center=<-20.7747 -25.3247 -3.2737>, rotation=<rot 0 88.3993 -85.591>}]
 W2B6-R [{shapeType=Capsule, length=61.9701, radius=28.5377, center=<20.7748 -25.3247 -3.2737>, rotation=<rot 0 -88.3993 -85.591>}]
 W2B7-L [{shapeType=Capsule, length=37.1471, radius=7.8746, center=<7.6674 -45.1447 -1.5173>, rotation=<rot 0 169.4667 -87.2826>}]
 W2B7-R [{shapeType=Capsule, length=37.1471, radius=7.8746, center=<-7.6673 -45.1447 -1.5173>, rotation=<rot 0 -169.4667 -87.2826>}]
 W3B1-L [{shapeType=Capsule, length=84.3579, radius=58.5948, center=<12.8726 -59.3507 -2.8216>, rotation=<rot 0 -7.6621 -86.0385>}]
 W3B1-R [{shapeType=Capsule, length=84.3579, radius=58.5948, center=<-12.8725 -59.3507 -2.8216>, rotation=<rot 0 7.6621 -86.0385>}]
 W3B3-L [{shapeType=Capsule, length=104.1338, radius=32.8612, center=<15.6257 -58.0465 -2.9082>, rotation=<rot 0 -76.9271 -79.8111>}]
 W3B3-R [{shapeType=Capsule, length=104.1338, radius=32.8612, center=<-15.6256 -58.0465 -2.9082>, rotation=<rot 0 76.9271 -79.8111>}]
 W3B4-L [{shapeType=Capsule, length=96.0499, radius=8.809, center=<7.1286 -69.6466 -2.8123>, rotation=<rot 0 -90.16 -77.0729>}]
 W3B4-R [{shapeType=Capsule, length=96.0499, radius=8.809, center=<-7.1285 -69.6466 -2.8123>, rotation=<rot 0 90.16 -77.0729>}]
 W3B5-L [{shapeType=Capsule, length=85.5997, radius=32.7475, center=<-4.2056 -48.3958 0.0283>, rotation=<rot 0 -90.9956 -72.3961>}]
 W3B5-R [{shapeType=Capsule, length=85.5997, radius=32.7472, center=<4.2056 -48.3962 0.024>, rotation=<rot 0 90.9939 -72.396>}]
 W3B6-L [{shapeType=Capsule, length=63.7476, radius=32.8522, center=<1.8817 -51.9464 -1.0679>, rotation=<rot 0 -73.4249 -69.3296>}]
 W3B6-R [{shapeType=Capsule, length=63.7476, radius=32.8526, center=<-1.8806 -51.9469 -1.0811>, rotation=<rot 0 73.4235 -69.332>}]
 W3B7-L [{shapeType=Capsule, length=42.3247, radius=11.1048, center=<9.406 -47.583 -3.4267>, rotation=<rot 0 -20.8497 -82.7153>}]
 W3B7-R [{shapeType=Capsule, length=42.3247, radius=11.1049, center=<-9.4041 -47.5834 -3.4461>, rotation=<rot 0 20.8501 -82.7208>}]
}
constraints [
 {bone1Name=W2B1-R, bone2Name=Bone_008, swing1LimitDegrees=45, swing2LimitDegrees=45, twistLimitDegrees=45}
 {bone1Name=W2B2-R, bone2Name=W2B1-R, swing1LimitDegrees=45, swing2LimitDegrees=45, twistLimitDegrees=45}
 {bone1Name=W2B3-R, bone2Name=W2B2-R, swing1LimitDegrees=45, swing2LimitDegrees=45, twistLimitDegrees=45}
 {bone1Name=W2B4-R, bone2Name=W2B3-R, swing1LimitDegrees=45, swing2LimitDegrees=45, twistLimitDegrees=45}
 {bone1Name=W2B5-R, bone2Name=W2B4-R, swing1LimitDegrees=45, swing2LimitDegrees=45, twistLimitDegrees=45}
 {bone1Name=W2B6-R, bone2Name=W2B5-R, swing1LimitDegrees=45, swing2LimitDegrees=45, twistLimitDegrees=45}
 {bone1Name=W2B7-R, bone2Name=W2B6-R, swing1LimitDegrees=45, swing2LimitDegrees=45, twistLimitDegrees=45}
 {bone1Name=W3B1-R, bone2Name=Bone_008, swing1LimitDegrees=45, swing2LimitDegrees=45, twistLimitDegrees=45}
 {bone1Name=W3B3-R, bone2Name=W3B1-R, swing1LimitDegrees=45, swing2LimitDegrees=45, twistLimitDegrees=45}
 {bone1Name=W3B4-R, bone2Name=W3B3-R, swing1LimitDegrees=45, swing2LimitDegrees=45, twistLimitDegrees=45}
 {bone1Name=W3B5-R, bone2Name=W3B4-R, swing1LimitDegrees=45, swing2LimitDegrees=45, twistLimitDegrees=45}
 {bone1Name=W3B6-R, bone2Name=W3B5-R, swing1LimitDegrees=45, swing2LimitDegrees=45, twistLimitDegrees=45}
 {bone1Name=W3B7-R, bone2Name=W3B6-R, swing1LimitDegrees=45, swing2LimitDegrees=45, twistLimitDegrees=45}
 {bone1Name=W2B1-L, bone2Name=Bone_008, swing1LimitDegrees=45, swing2LimitDegrees=45, twistLimitDegrees=45}
 {bone1Name=W2B2-L, bone2Name=W2B1-L, swing1LimitDegrees=45, swing2LimitDegrees=45, twistLimitDegrees=45}
 {bone1Name=W2B3-L, bone2Name=W2B2-L, swing1LimitDegrees=45, swing2LimitDegrees=45, twistLimitDegrees=45}
 {bone1Name=W2B4-L, bone2Name=W2B3-L, swing1LimitDegrees=45, swing2LimitDegrees=45, twistLimitDegrees=45}
 {bone1Name=W2B5-L, bone2Name=W2B4-L, swing1LimitDegrees=45, swing2LimitDegrees=45, twistLimitDegrees=45}
 {bone1Name=W2B6-L, bone2Name=W2B5-L, swing1LimitDegrees=45, swing2LimitDegrees=45, twistLimitDegrees=45}
 {bone1Name=W2B7-L, bone2Name=W2B6-L, swing1LimitDegrees=45, swing2LimitDegrees=45, twistLimitDegrees=45}
 {bone1Name=W3B1-L, bone2Name=Bone_008, swing1LimitDegrees=45, swing2LimitDegrees=45, twistLimitDegrees=45}
 {bone1Name=W3B3-L, bone2Name=W3B1-L, swing1LimitDegrees=45, swing2LimitDegrees=45, twistLimitDegrees=45}
 {bone1Name=W3B4-L, bone2Name=W3B3-L, swing1LimitDegrees=45, swing2LimitDegrees=45, twistLimitDegrees=45}
 {bone1Name=W3B5-L, bone2Name=W3B4-L, swing1LimitDegrees=45, swing2LimitDegrees=45, twistLimitDegrees=45}
 {bone1Name=W3B6-L, bone2Name=W3B5-L, swing1LimitDegrees=45, swing2LimitDegrees=45, twistLimitDegrees=45}
 {bone1Name=W3B7-L, bone2Name=W3B6-L, swing1LimitDegrees=45, swing2LimitDegrees=45, twistLimitDegrees=45}
 {bone1Name=W1B2-R, bone2Name=W1B1-R, swing1LimitDegrees=45, swing2LimitDegrees=45, twistLimitDegrees=45}
 {bone1Name=W1B3-R, bone2Name=W1B2-R, swing1LimitDegrees=45, swing2LimitDegrees=45, twistLimitDegrees=45}
 {bone1Name=W1B4-R, bone2Name=W1B3-R, swing1LimitDegrees=45, swing2LimitDegrees=45, twistLimitDegrees=45}
 {bone1Name=W1B5-R, bone2Name=W1B4-R, swing1LimitDegrees=45, swing2LimitDegrees=45, twistLimitDegrees=45}
 {bone1Name=W1B6-R, bone2Name=W1B5-R, swing1LimitDegrees=45, swing2LimitDegrees=45, twistLimitDegrees=45}
 {bone1Name=W1B2-L, bone2Name=W1B1-L, swing1LimitDegrees=45, swing2LimitDegrees=45, twistLimitDegrees=45}
 {bone1Name=W1B3-L, bone2Name=W1B2-L, swing1LimitDegrees=45, swing2LimitDegrees=45, twistLimitDegrees=45}
 {bone1Name=W1B4-L, bone2Name=W1B3-L, swing1LimitDegrees=45, swing2LimitDegrees=45, twistLimitDegrees=45}
 {bone1Name=W1B5-L, bone2Name=W1B4-L, swing1LimitDegrees=45, swing2LimitDegrees=45, twistLimitDegrees=45}
 {bone1Name=W1B6-L, bone2Name=W1B5-L, swing1LimitDegrees=45, swing2LimitDegrees=45, twistLimitDegrees=45}
]
